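# Masked, two-sided Default Lit material for golden Niagara sprite, ribbon and mesh particles on the Jail Bird Bumble back bling.
# Particle color feeds the MF_FX_EmissiveAdjustments function, alongside parameters for a gold base color and roughness.
# A procedural UV gradient multiplies squared coordinate and inverted-coordinate components, fading toward the edges without a texture.
# The graph does not show which material outputs these nodes drive.
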Material /BRCosmetics/Accessories/FORT_Backpacks/Backpack_F_MED_Jail_Bird_Bumble/FX/M_BackPack_JailBirdBumble_GoldenParticle
graph {
 properties {
  bUsedWithNiagaraMeshParticles=true
  bUsedWithNiagaraRibbons=true
  bUsedWithNiagaraSprites=true
  blendMode=BLEND_Masked
  opacityMaskClipValue=0.01
  pixelDepthOffsetMode=PDOM_Legacy
  twoSided=true
  usedShadingModels=Default Lit
 }
 nodes [
  Constant() {r=1}
  MaterialFunctionCall(Emissive=ParticleColor().RGB) {materialFunction=&MF_FX_EmissiveAdjustments}
  n1 = Multiply(A=TextureCoordinate())
  n2 = MaterialFunctionCall(Float2=@n1) {materialFunction=&BreakOutFloat2Components}
  n3 = MaterialFunctionCall(Float2=OneMinus(None=@n1)) {materialFunction=&BreakOutFloat2Components}
  Multiply(A=Multiply(A=@n2.R, B=@n2.R), B=Multiply(A=@n3.R, B=@n3.R))
  ScalarParameter() {defaultValue=0.2648, parameterName=Roughness}
  VectorParameter() {parameterName=BaseColor, defaultValue=<rgba 0.728 0.4139 0.0946 1>}
 ]
}
