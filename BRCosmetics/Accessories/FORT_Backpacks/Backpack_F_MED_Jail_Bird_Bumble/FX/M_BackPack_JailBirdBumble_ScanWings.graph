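# Translucent two-sided material for Niagara mesh particles on the Jail Bird Bumble back bling, drawing a scanning band across the wings.
# A time-driven sine shifts UV1's U coordinate, and a custom range clip plus a periodic sine shape the band.
# The depth-faded band is multiplied by a mask texture, inverted radial gradient and particle dynamic parameter.
# A separate UV0 vertical gradient blends dark and high colors.
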Material /BRCosmetics/Accessories/FORT_Backpacks/Backpack_F_MED_Jail_Bird_Bumble/FX/M_BackPack_JailBirdBumble_ScanWings
graph {
 properties {
  bUsedWithNiagaraMeshParticles=true
  blendMode=BLEND_Translucent
  pixelDepthOffsetMode=PDOM_Legacy
  twoSided=true
  usedShadingModels=Default Lit
 }
 nodes [
  n1 = TextureCoordinate() {coordinateIndex=1}
  n2 = ScalarParameter() {defaultValue=0.0848, parameterName=RemapTimeLowValue}
  n3 = ScalarParameter() {defaultValue=-0.7475, parameterName=RemapTimeHighValue}
  n4 = Multiply(A=Time() {bOverride_Period=true, period=55}, B=ScalarParameter() {defaultValue=368.3365, parameterName=TimePeriod})
  n5 = Multiply(A=Add(A=Sine(None=@n4))) {constB=0.5}
  MaterialExpressionMultiply_0 = Multiply(
   A=Add(
    A=MaterialFunctionCall(Float2=@n1) {materialFunction=&BreakOutFloat2Components}.R
    B=LinearInterpolate(A=@n2, B=@n3, Alpha=@n5)
   )
   B=ScalarParameter() {defaultValue=3.8894, parameterName=UVScale}
  )
  n6 = Custom!(A=@MaterialExpressionMultiply_0) {
   code="float B;\r\nif(A>=0&&&&A<=3){\r\nB = A;\r\n}else{\r\nB = 0;\r\n}\r\nreturn B;"
   outputType=CMOT_Float1
   inputs [{
     inputName=A
     input {expression=&self:MaterialExpressionMultiply_0}
    }]
  }
  n7 = Saturate(None=Ceil(None=@n6))
  n8 = Sine(None=@MaterialExpressionMultiply_0) {period=5}
  n9 = DepthFade(
   Opacity=Multiply(A=Multiply(A=@n7, B=@n8), B=ScalarParameter() {defaultValue=0.0791, parameterName=Sharpness})
   FadeDistance=ScalarParameter() {defaultValue=22, parameterName=DepthFade}
  )
  n10 = ScalarParameter() {defaultValue=3.054, parameterName=ScanStrMultiply}
  n11 = TextureSample() {
   samplerType=SAMPLERTYPE_LinearColor
   texture=&/T_JailBirdBumble_Mask
  }
  n12 = MaterialFunctionCall(
   Radius=ScalarParameter() {defaultValue=0.4, parameterName=SphereGradientRadius}
   Density=ScalarParameter() {defaultValue=4, parameterName=SphereGradientDensity}
  ) {materialFunction=&RadialGradientExponential}
  n13 = VectorParameter() {parameterName=SphereGradientRemap, defaultValue=<rgba 0 1 0.4 1>}
  n14 = MaterialFunctionCall(
   Input=Saturate(None=OneMinus(None=@n12.RadialGradientExponential))
   Input Low=@n13.R
   Input High=@n13.R
   Target Low=@n13.R
   Target High=@n13.R
  ) {materialFunction=&RemapValueRange}
  Multiply(A=Multiply(A=Multiply(A=@n9, B=@n10), B=Multiply(A=@n11.RGB, B=@n14.Result)), B=DynamicParameter().Param1)
  n15 = TextureCoordinate()
  n16 = Power(
   Base=OneMinus(None=MaterialFunctionCall(Float2=@n15) {materialFunction=&BreakOutFloat2Components}.G)
   Exp=ScalarParameter() {defaultValue=2, parameterName=ColorGradientPower}
  )
  Reroute(None=LinearInterpolate(
    A=VectorParameter() {parameterName=DarkColor, defaultValue=<rgba 0.2708 0.2031 0 1>}.RGB
    B=VectorParameter() {parameterName=HighColor, defaultValue=<rgba 1 1 1 1>}.RGB
    Alpha=OneMinus(None=@n16)
   ))
 ]
}
