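# Niagara system for the Jail Bird Bumble backpack cosmetic, attaching two mesh particles to the backpack skeletal mesh.
# Two single bursts use spawn group to pick each particle's bone or socket, and negative scale mirrors one copy.
# Chained scratch modules apply fixed quaternion offsets to sampled rotation, while User.IsJitter feeds a dynamic material parameter.
# A disabled second emitter would spawn drifting, tumbling sprite and mesh particles at a rate scaled by User.IsJump.
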NiagaraSystem /BRCosmetics/Accessories/FORT_Backpacks/Backpack_F_MED_Jail_Bird_Bumble/FX/NS_BackPack_JailBirdBumble_V1
dependencies {usedDataInterfaces [&NiagaraDataInterfaceCurve], usedDynamicInputs [], usedModules [&SystemState]}
emitters {
 Empty {
  emitterUpdateScript [
   &EmitterState(
    Inactive Response=::NewEnumerator1
    Life Cycle Mode=::NewEnumerator1
    Loop Behavior=::NewEnumerator1
    Loop Duration Mode=::NewEnumerator1
    UseLoopDelay=false
   )
   &SpawnBurst_Instantaneous(Spawn Count=1, Use Spawn Probability=false)
   &SpawnBurst_Instantaneous(Spawn Count=1, Spawn Group=1, Use Spawn Probability=false) {module=SpawnBurst_Instantaneous001}
  ]
  particleSpawnScript [&InitializeParticle(
    Color=&RandomRangeLinearColor(Minimum=<rgba 0.065 0.065 0.065 1>, Maximum=<rgba 1 1 1 1>)
    UsePositionOffset=false
    Write Material Random=true
   )]
  particleUpdateScript [
   &ParticleState(Kill Particles When Lifetime Has Elapsed=false, Let Infinitely Lived Particles Die When Emitter Deactivates=true)
   &SkeletalMeshLocation(
    Apply Sampled Orientation to Mesh Particles=true
    Bone / Socket Index=$Particles.SpawnGroup
    Bone Sampling Mode=::NewEnumerator4
    Inherit Velocity=false
    Interpolate Mesh Sampling=false
    MaskSpawnGroup=false
    Sampled Position Offset Coordinate Space:ENiagaraMeshTransforms=::NewEnumerator2
    Skeletal Mesh=$User.SKM
   ) {enabled=false}
   &SkeletalMeshLocation(Bone / Socket Index=1, Bone Sampling Mode=::NewEnumerator4, MaskSpawnGroup=false, Skeletal Mesh=$User.SKM) {enabled=false, module=SkeletalMeshLocation001}
   &DynamicMaterialParameters(
    Index 0 Param 1=$User.IsJitter
    Module.Param0WriteEnabled=true
    Module.Param1WriteEnabled=true
    Module.Param2WriteEnabled=true
    Module.Param3WriteEnabled=true
   )
   &SkeletalMeshLocation(
    Bone / Socket Index=&SpawnGroup()
    Bone Sampling Mode=::NewEnumerator4
    MaskSpawnGroup=false
    Skeletal Mesh=$User.SKM
   ) {module=SkeletalMeshLocation002}
   &self:ScratchModule(New Quat:Quat4f=&MultiplyQuaternion(
     Quaternion A=$Particles.SkeletalMeshLocation002.SampledRotation
     Quaternion B=&MakeQuatFromAxisAngle(Axis=<1 0 0>, AngleInDegrees=-161.3798)
    )) {module=ScratchModule}
   &self:ScratchModule(New Quat:Quat4f=&MultiplyQuaternion(
     Quaternion A=$Particles.MeshOrientation
     Quaternion B=&MakeQuatFromAxisAngle(Axis=<0 0 1>, AngleInDegrees=-39.2163)
    )) {module=ScratchModule001}
   &self:ScratchModule(New Quat:Quat4f=&MultiplyQuaternion(
     Quaternion A=$Particles.MeshOrientation
     Quaternion B=&MakeQuatFromAxisAngle(Axis=<0 1 0>, AngleInDegrees=-29.8761)
    )) {module=ScratchModule002}
   /SetVariables(Particles.Scale=&MakeCustomVectorFromBool(
     Bool=&SetBoolByIntComparison(A=&SpawnGroup(), B=1, Comparison Type=::NewEnumerator4)
     True Vector=<-0.74 -0.74 -0.74>
     False Vector=<0.74 0.74 0.74>
    )) {isSetParameters=true}
  ]
  rendererClasses [&NiagaraMeshRendererProperties]
  renderers [{rendererClass=&NiagaraMeshRendererProperties, rendererIndex=0}]
 }
 Empty003 {
  bEnabled=false
  emitterUpdateScript [
   &EmitterState()
   &SpawnRate(SpawnRate=&Multiply_Float(A=33, B=$User.IsJump), Use Spawn Probability=false)
  ]
  particleSpawnScript [
   &InitializeParticle(
    Color=<rgba 0 0 0 1>
    Lifetime Max=2.1003
    Lifetime Min=0.2864
    Lifetime Mode=::NewEnumerator1
    Sprite Size Mode=::NewEnumerator3
    Uniform Sprite Size=1
    UsePositionOffset=false
    Write Material Random=true
   )
   &SkeletalMeshLocation(
    Bone Sampling Mode=::NewEnumerator6
    Inherit Velocity=false
    Interpolate Mesh Sampling=false
    MaskSpawnGroup=false
    Sampled Position Offset Coordinate Space:ENiagaraMeshTransforms=::NewEnumerator3
    Skeletal Mesh=$User.SKM
   )
   &SphereLocation(
    Hemisphere Y=true
    MaskSpawnGroup=false
    Non Uniform Scale=<1 1 0.56>
    Offset=<0 -7.282 8.7406>
    Override Seed=false
    Sphere Orientation Axis:Vector3f=&NormalizeVector(Vector To Normalize=<1 0 0>)
    Sphere Radius=10.8861
    Surface Only=true
    UseNonUniformScale=true
    UseOffset=true
    UseSphereOrientation=true
   )
   /Niagara/Modules/Spawn/Location/ConeLocation(MaskSpawnGroup=false, Module.Override Randomness=false) {enabled=false}
   &AddVelocityInCone(
    Cone Angle=92.7
    Cone Axis=<0 -1 0>
    Module.Use Velocity Falloff On Cone Axis=false
    Velocity Distribution Along Cone Axis=0
    Velocity Strength=88.2263
   ) {enabled=false}
   &InitialMeshOrientation(
    Rotation=&RandomRangeVector(Minimum=<-1 -1 -1>, Maximum=<1 1 1>)
    Rotation Coordinate Space=::NewEnumerator2
    Use Orientation Vector=true
    Use Rotation Vector=true
   )
   &AddVelocityFromPoint(
    Module.Use Falloff Distance=false
    Origin Offset=<0 5 8>
    Velocity Strength=&RandomRangeFloat(Minimum=33.5896, Maximum=50.1367)
   )
   &AddVelocity(Coordinate Space=::World, Velocity=&RandomRangeVector(Minimum=<0 0 -15.6626>, Maximum=<0 0 -27.9767>))
   /SetVariables(Particles.NewVector:Vector3f=$Particles.Velocity) {isSetParameters=true}
   &ApplyInitialForces()
  ]
  particleUpdateScript [
   /SetVariables(Particles.Velocity=$Particles.NewVector) {isSetParameters=true}
   &Drag(Drag=1.3441, Use Linear Drag=true, Use Rotational Drag=false)
   &CurlNoiseForce(Noise Frequency=20.948, Noise Strength=181.5907, Pan Noise=true, Pan Noise Field=<1 1 0>) {enabled=false}
   &SolveForcesAndVelocity(Clamp Velocity=false, Limit Acceleration=false, Speed Limit=333)
   &ParticleState()
   /SetVariables(Particles.Scale=&VectorFromFloat(Value=&RandomRangeFloat(Minimum=2, Maximum=2))) {isSetParameters=true}
   &UpdateMeshOrientation(Rotation Vector=&RandomVector(Vector Scale=1))
   &UpdateMeshOrientation(Rotation Vector=&RandomVector(Vector Scale=1)) {module=UpdateMeshOrientation001}
   &UpdateMeshOrientation(Rotation Vector=&RandomVector(Vector Scale=1)) {module=UpdateMeshOrientation002}
   /SetVariables(Particles.Scale=&Multiply_VectorByFloat(
     Vector=$Particles.Scale
     Float=&FloatFromCurve(FloatCurve=DataInterface() {propertyValues {Curve {keys [
          {arriveTangent=0.2821, interpMode=RCIM_Cubic, leaveTangent=0.2821, tangentMode=RCTM_User, time=0, value=0}
          {arriveTangent=0.1722, interpMode=RCIM_Cubic, leaveTangent=0.1722, tangentMode=RCTM_User, time=0.4643, value=0.9996}
          {interpMode=RCIM_Cubic, leaveTangent=-3.7884, tangentMode=RCTM_Break, time=0.7602, value=0.9225}
          {arriveTangent=-3.7819, interpMode=RCIM_Cubic, leaveTangent=-3.7819, tangentMode=RCTM_User, time=1, value=0}
         ]}}})
    )) {isSetParameters=true}
   /SetVariables(Particles.Position=&Add_Vector(
     A=Linked() {linkedVariable {name=Particles.Position, type=NiagaraPosition}}
     B=&Multiply_VectorByFloat(
      Vector=&Multiply_VectorByFloat(
       Vector=/Niagara/DynamicInputs/Helpers/SafeNormalizeVector(Vector To Normalize:Vector3f=$Engine.Owner.Velocity, Fallback Vector:Vector3f=<0 0 0.1>)
       Float=&ClampFloat(Float=/Niagara/DynamicInputs/TypeConversions/VectorLengthSafe(Vector:Vector3f=$Engine.Owner.Velocity), Max=1051.1453)
      )
      Float=$Engine.DeltaTime
     )
    )) {enabled=false, isSetParameters=true}
   /SetVariables(Particles.NewVector:Vector3f=$Particles.Velocity) {isSetParameters=true}
   /SetVariables(Particles.Velocity=&Lerp_Vector(
     A=&Multiply_VectorByFloat(Vector=$Particles.NewVector, Float=1)
     B=&Multiply_VectorByFloat(Vector=$Particles.NewVector, Float=11)
     Alpha=&FloatFromCurve(FloatCurve=DataInterface() {propertyValues {Curve {keys [{tangentMode=RCTM_User, time=0, value=1}, <at 0.4197 0.0175>, {tangentMode=RCTM_User, time=1, value=0}]}}})
    )) {isSetParameters=true}
   &GenerateLocationEvent(
    Module.Use Event Send Rate=true
    Position To Send:Vector3f=/Niagara/DynamicInputs/Transforms/TransformPosition(
     Position=Linked() {linkedVariable {name=Particles.Position, type=NiagaraPosition}}
     Source Space=::World
     Destination Space=::Local
     Apply Scale=true
    )
   )
  ]
  rendererClasses [&NiagaraSpriteRendererProperties, &NiagaraMeshRendererProperties]
  renderers [
   {rendererClass=&NiagaraSpriteRendererProperties}
   {rendererClass=&NiagaraMeshRendererProperties, rendererIndex=1}
  ]
 }
}
summary {
 systemName=NS_BackPack_JailBirdBumble_V1
 emitters [
  {emitterName=Empty, rendererClasses [&NiagaraMeshRendererProperties]}
  {
   bEnabled=false
   emitterName=Empty003
   rendererClasses [&NiagaraSpriteRendererProperties, &NiagaraMeshRendererProperties]
  }
 ]
 userVariables [
  User.IsJitter {type=NiagaraFloat, defaultValue=0, description=""}
  User.IsJump {type=NiagaraFloat, defaultValue=0, description=""}
  User.SKM {
   type=NiagaraDataInterfaceSkeletalMesh
   defaultValue {
    struct=NiagaraExt_VariableValue_DataInterface
    value {dataInterface=&self:NiagaraDataInterfaceSkeletalMesh_0}
   }
  }
 ]
}
