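# Physics asset for the Jail Bird Bumble backpack cosmetic, providing simulation bodies for its left and right wings.
# Each wing is a four-bone chain of capsules hanging off a small capsule on the pack bone.
# Constraints link each segment to its parent with very tight swing and twist limits that loosen slightly toward the tips, so the wings flex only subtly.
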PhysicsAsset /BRCosmetics/Accessories/FORT_Backpacks/Backpack_F_MED_Jail_Bird_Bumble/Meshes/F_MED_Jail_Bird_Bumble_Pack_Physics
bodies {
 dyn_wing_l [{shapeType=Capsule, length=5, radius=4, center=<4 0 0>, rotation=<rot -90 0 0>}]
 dyn_wing_l_01 [{shapeType=Capsule, length=1, radius=5, center=<4 0 0>, rotation=<rot -90 0 0>}]
 dyn_wing_l_02 [{shapeType=Capsule, length=1, radius=5, shapeName=dyn_wing_l_02_capsule, center=<4 0 0>, rotation=<rot -90 0 0>}]
 dyn_wing_l_03 [{shapeType=Capsule, length=1, radius=5, center=<4 0 0>, rotation=<rot -90 0 0>}]
 dyn_wing_r [{shapeType=Capsule, length=5, radius=4, shapeName=dyn_wing_r_capsule, center=<-4 0 0>, rotation=<rot -90 0 0>}]
 dyn_wing_r_01 [{shapeType=Capsule, length=1, radius=5, shapeName=dyn_wing_r_01_capsule, center=<-4 0 0>, rotation=<rot -90 0 0>}]
 dyn_wing_r_02 [{shapeType=Capsule, length=1, radius=5, shapeName=dyn_wing_r_02_capsule, center=<-4 0 0>, rotation=<rot -90 0 0>}]
 dyn_wing_r_03 [{shapeType=Capsule, length=1, radius=5, shapeName=dyn_wing_r_03_capsule, center=<-4 0 0>, rotation=<rot -90 0 0>}]
 pack [{shapeType=Capsule, length=1, radius=1, shapeName=pack_capsule}]
}
constraints [
 {bone1Name=dyn_wing_l_01, bone2Name=dyn_wing_l, swing1LimitDegrees=3, swing2LimitDegrees=3, twistLimitDegrees=1}
 {bone1Name=dyn_wing_r, bone2Name=pack, swing1LimitDegrees=2, swing2LimitDegrees=2, twistLimitDegrees=1}
 {bone1Name=dyn_wing_l, bone2Name=pack, swing1LimitDegrees=2, swing2LimitDegrees=2, twistLimitDegrees=1}
 {bone1Name=dyn_wing_l_02, bone2Name=dyn_wing_l_01, swing1LimitDegrees=5, swing2LimitDegrees=5, twistLimitDegrees=1}
 {bone1Name=dyn_wing_l_03, bone2Name=dyn_wing_l_02, swing1LimitDegrees=5, swing2LimitDegrees=5, twistLimitDegrees=1}
 {bone1Name=dyn_wing_r_01, bone2Name=dyn_wing_r, swing1LimitDegrees=3, swing2LimitDegrees=3, twistLimitDegrees=1}
 {bone1Name=dyn_wing_r_02, bone2Name=dyn_wing_r_01, swing1LimitDegrees=5, swing2LimitDegrees=5, twistLimitDegrees=1}
 {bone1Name=dyn_wing_r_03, bone2Name=dyn_wing_r_02, swing1LimitDegrees=5, swing2LimitDegrees=5, twistLimitDegrees=1}
]
